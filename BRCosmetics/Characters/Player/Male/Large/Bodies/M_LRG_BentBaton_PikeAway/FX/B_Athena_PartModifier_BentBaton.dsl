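; Cosmetic part modifier for the BentBaton (PikeAway) character that overrides the body's subsurface color in frontend displays.
; On post-res-in, after calling the parent, it gets the pawn's Body and Head meshes and matches their first material's name against the PikeAway body or face accessory materials.
; When either matches in a frontend context, it pushes its SSS Color variable to the "SSS Color" parameter on the Body mesh materials only.
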(event EventOnPostResIn
  (bind _returnvalue (CustomCharacter|GetAssociatedPlayerPawn))
  (bind _returnvalue_2 (Customization|GetSkeletalMeshforPartType _returnvalue "Body" :!class /Script/FortniteGame.FortPlayerPawn))
  (|CallParentFunction)
  (if (and (Reactive|Frontend|IsCosmeticObjectDisplayedinFrontendContext 0) (or (Utilities|String|Contains (Utilities|GetDisplayName (Rendering|Material|GetMaterial _returnvalue_2 0)) "BentBaton_PikeAway_Body" false true) (Utilities|String|Contains (Utilities|GetDisplayName (Rendering|Material|GetMaterial (Customization|GetSkeletalMeshforPartType _returnvalue "Head" :!class /Script/FortniteGame.FortPlayerPawn) 0)) "BentBaton_PikeAway_FaceAcc" false true)))
    (Rendering|Material|SetVectorParameterValueonMaterials _returnvalue_2 "SSS Color" (Math|Conversions|ToVector(LinearColor) (Variables|Default|GetSSS%20Color)))))

(fn ConstructionScript
  (|CallParentFunction))
