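# Parent material for the BentBaton PikeAway large male player character body, using subsurface shading and material attributes on skeletal meshes with morph target and mesh deformer support.
# The shared 2023 character PreFX function builds the base, which the character-specific MF_BentBaton function modifies before the shared PostFX function finishes it.
# PreFX also feeds PostFX directly with data such as pre-skinned local position, distance blend and fresnel.
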Material /BRCosmetics/Characters/Player/Male/Large/Bodies/M_LRG_BentBaton_PikeAway/Materials/M_BentBaton_Parent
graph {
 properties {
  bHasPixelAnimation=true
  bUseMaterialAttributes=true
  bUsedWithMeshDeformer=true
  bUsedWithMorphTargets=true
  bUsedWithSkeletalMesh=true
  pixelDepthOffsetMode=PDOM_Legacy
  shadingModel=MSM_Subsurface
  usedShadingModels=Subsurface
 }
 nodes [
  n1 = MaterialFunctionCall() {materialFunction=&MF_CharacterParent_2023_PreFX}
  MaterialFunctionCall(
   Material=MaterialFunctionCall!(M=@n1.M, Result=@n1.M, Tech Art Mask=@n1.M) {materialFunction=&/MF_BentBaton}.SetMaterialAttributes
   M=@n1.M
   Pre Skinned Local Position=@n1.M
   Distance Blend=@n1.M
   Fresnel=@n1.M
  ) {materialFunction=&MF_CharacterParent_2023_PostFX}
 ]
}
