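# Control Rig for the large male BentBaton_PikeAway body, driving extra "woolf" digitigrade-style leg and toe bones from the humanoid skeleton.
# After the shared DeformRig function, Leg Setup copies thigh/calf rotations, solves PBIK to each foot, adds pose-reader-weighted rotation offsets and aims woolf feet.
# ToeSetup aim-constrains toes to ball controls, then adjusts them by foot yaw delta.
# Leg Setup also holds unwired alternatives, including parent constraints.
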ControlRigBlueprint /BRCosmetics/Characters/Player/Male/Large/Bodies/M_LRG_BentBaton_PikeAway/Meshes/M_LRG_BentBaton_PikeAway_CtrlRig
graphs [
 New Function_ContainedGraph {nodes [
   n1 = RigVMFunctionEntryNode()
   RigVMFunction_Sequence = &RigVMFunction_Sequence(ExecuteContext=@n1.ExecuteContext)
   RigVMFunctionReturnNode(ExecuteContext=@RigVMFunction_Sequence.B)
   SequenceAggregate = RigVMAggregateNode(ExecuteContext=@RigVMFunction_Sequence.A)
   RigUnit_ItemArray = &RigUnit_ItemArray(Items=((Type=Bone,Name="thigh_l"),(Type=Bone,Name="calf_l"),(Type=Bone,Name="calf_l")))
   DISPATCH_RigVMDispatch_ArrayIterator = RigVMDispatchNode(Array=@RigUnit_ItemArray.Items, ExecuteContext=@SequenceAggregate.A) {template="DISPATCH_RigVMDispatch_ArrayIterator(in Array,out Element,out Index,out Count,out Ratio)"}
   DISPATCH_RigVMDispatch_ArrayGetAtIndex = RigVMDispatchNode(
    Array=&RigUnit_ItemArray(Items=((Type=Bone,Name="woolf_thigh_l"),(Type=Bone,Name="woolf_calf_l"),(Type=Bone,Name="woolf_foot_l"))).Items
    Index=@DISPATCH_RigVMDispatch_ArrayIterator.Index
   ) {template="DISPATCH_RigVMDispatch_ArrayGetAtIndex(in Array,in Index,out Element)"}
   > &RigUnit_SetRotation(
    Space=GlobalSpace
    Weight=1
    bInitial=False
    bPropagateToChildren=true
    Item=@DISPATCH_RigVMDispatch_ArrayGetAtIndex.Element
    Value=&RigUnit_GetTransform(Item=@DISPATCH_RigVMDispatch_ArrayIterator.Element).Transform.Rotation
   )
   PBIK = /Script/PBIK.RigUnit_PBIK(
    BoneSettings="((Bone=\"woolf_calf_l\",X=Locked,Y=Locked,Z=Free,MinZ=0,MaxZ=0,bUsePreferredAngles=false,PreferredAngles=(X=0,Y=0,Z=0)),(Bone=\"woolf_foot_l\",RotationStiffness=0,PositionStiffness=0,X=Locked,MinX=0,MaxX=0,Y=Locked,MinY=0,MaxY=0,Z=Free,MinZ=0,MaxZ=0,bUsePrefer ...(truncated)"
    Debug=(DrawScale=1,bDrawDebug=false)
    Effectors=((Bone="woolf_ball_l",Transform=(Rotation=(X=0,Y=0,Z=0,W=1),Translation=(X=0,Y=0,Z=0),Scale3D=(X=1,Y=1,Z=1)),PositionAlpha=1,RotationAlpha=1,StrengthAlpha=1,PullChainAlpha=0,PinRotation=1,ChainDepth=0))
    ExcludedBones=()
    Root=woolf_thigh_l
    Settings="(Iterations=20,SubIterations=0,MassMultiplier=1,bAllowStretch=False,RootBehavior=PinToInput,PrePullRootSettings=(RotationAlpha=0,RotationAlphaX=1,RotationAlphaY=1,RotationAlphaZ=1,PositionAlpha=1,PositionAlphaX=1,PositionAlphaY=1,PositionAlphaZ=1),GlobalPu ...(truncated)"
    Effectors.0.Transform=&RigUnit_GetTransform(Item=(Type=Bone,Name="foot_l")).Transform
    ExecutePin=@DISPATCH_RigVMDispatch_ArrayIterator.Completed
   ) {is=struct RigVMUnitNode}
   > n2 = &RigUnit_SphericalPoseReader(
    ActiveRegionScaleFactors=(PositiveWidth=4,NegativeWidth=3,PositiveHeight=2.5,NegativeHeight=2)
    ActiveRegionSize=0.1
    Debug=(bDrawDebug=false,bDraw2D=false,bDrawLocalAxes=false,DebugScale=25,DebugSegments=20,DebugThickness=0.25)
    DriverAxis=(X=-1,Y=0,Z=0)
    DriverItem=(Type=Bone,Name="thigh_l")
    FalloffRegionScaleFactors=(PositiveWidth=1,NegativeWidth=1,PositiveHeight=1,NegativeHeight=0.5)
    FalloffSize=0.4
    OptionalParentItem=(Type=Bone,Name="pelvis")
    RotationOffset=(X=100,Y=0,Z=-20)
   )
   > &RigUnit_ModifyTransforms(
    ItemToModify=((Item=(Type=Bone,Name="woolf_thigh_l"),Transform=(Rotation=(X=0,Y=0,Z=-0.0872,W=0.9962),Translation=(X=0,Y=0,Z=0),Scale3D=(X=1,Y=1,Z=1))))
    WeightMinimum=0
    Weight=@n2.OutputParam
   )
   > &RigUnit_ModifyTransforms(
    ItemToModify=((Item=(Type=Bone,Name="woolf_calf_l"),Transform=(Rotation=(X=0,Y=0,Z=-0.0872,W=0.9962),Translation=(X=0,Y=0,Z=0),Scale3D=(X=1,Y=1,Z=1))))
    WeightMinimum=0
    Weight=@n2.OutputParam
   )
   n3 = /Script/ControlRig.RigUnit_AimBoneMath(
    DebugSettings=(bEnabled=False,Scale=10,WorldOffset=(Rotation=(X=0,Y=0,Z=0,W=1),Translation=(X=0,Y=0,Z=0),Scale3D=(X=1,Y=1,Z=1)))
    Primary=(Weight=1,Axis=(X=1,Y=0,Z=0),Target=(X=1,Y=0,Z=0),Kind=Location,Space=(Type=Bone,Name="None"))
    Secondary=(Weight=1,Axis=(X=0,Y=-1,Z=0),Target=(X=0,Y=0,Z=0),Kind=Location,Space=(Type=Bone,Name="None"))
    Weight=1
    InputTransform=&RigUnit_GetTransform(Item=(Type=Bone,Name="woolf_foot_l"), bInitial=False).Transform
    Primary.Target=&RigUnit_GetTransform(Item=(Type=Bone,Name="foot_l"), bInitial=False).Transform.Translation
    Secondary.Target=&RigUnit_GetTransform(Item=(Type=Bone,Name="thigh_aim_l"), bInitial=False).Transform.Translation
   ) {is=struct RigVMUnitNode}
   &RigUnit_SetTransform(
    Item=(Type=Bone,Name="woolf_foot_l")
    Weight=1
    bInitial=False
    Value=/Script/RigVM.RigVMFunction_MathTransformMakeAbsolute(Local=(Rotation=(X=0,Y=0,Z=0,W=1),Translation=(X=0,Y=0,Z=0),Scale3D=(X=1,Y=1,Z=1)), Parent=@n3.Result) {is=struct RigVMUnitNode, template="Make Absolute::Execute(in Local,in Parent,out Global)"}.Global
    ExecutePin=@SequenceAggregate.C
   )
   n4 = &RigUnit_AimBoneMath(
    DebugSettings=(bEnabled=False,Scale=10,WorldOffset=(Rotation=(X=0,Y=0,Z=0,W=1),Translation=(X=0,Y=0,Z=0),Scale3D=(X=1,Y=1,Z=1)))
    Primary=(Weight=1,Axis=(X=1,Y=0,Z=0),Target=(X=1,Y=0,Z=0),Kind=Location,Space=(Type=Bone,Name="None"))
    Secondary=(Weight=1,Axis=(X=0,Y=-1,Z=0),Target=(X=1,Y=0,Z=0),Kind=Location,Space=(Type=Bone,Name="None"))
    Weight=1
    Secondary.Target=&RigUnit_GetTransform(Item=(Type=Bone,Name="thigh_aim_r"), bInitial=False).Transform.Translation
    InputTransform=&RigUnit_GetTransform(Item=(Type=Bone,Name="woolf_foot_r"), bInitial=False).Transform
    Primary.Target=&RigUnit_GetTransform(Item=(Type=Bone,Name="foot_r"), bInitial=False).Transform.Translation
   )
   > &RigUnit_SetTransform(
    Item=(Type=Bone,Name="woolf_foot_r")
    Weight=1
    bInitial=False
    Value=&RigVMFunction_MathTransformMakeAbsolute(Local=(Rotation=(X=0,Y=0,Z=0,W=1),Translation=(X=0,Y=0,Z=0),Scale3D=(X=1,Y=1,Z=1)), Parent=@n4.Result) {template="Make Absolute::Execute(in Local,in Parent,out Global)"}.Global
   )
   n5 = RigVMDispatchNode(
    Array=&RigUnit_ItemArray(Items=((Type=Bone,Name="thigh_r"),(Type=Bone,Name="calf_r"),(Type=Bone,Name="calf_r"))).Items
    ExecuteContext=@SequenceAggregate.B
   ) {template="DISPATCH_RigVMDispatch_ArrayIterator(in Array,out Element,out Index,out Count,out Ratio)"}
   > &RigUnit_SetRotation(
    Space=GlobalSpace
    Weight=1
    bInitial=False
    bPropagateToChildren=true
    Item=RigVMDispatchNode(
     Array=&RigUnit_ItemArray(Items=((Type=Bone,Name="woolf_thigh_r"),(Type=Bone,Name="woolf_calf_r"),(Type=Bone,Name="woolf_foot_r"))).Items
     Index=@n5.Index
    ) {template="DISPATCH_RigVMDispatch_ArrayGetAtIndex(in Array,in Index,out Element)"}.Element
    Value=&RigUnit_GetTransform(Item=@n5.Element).Transform.Rotation
   )
   &RigUnit_PBIK(
    BoneSettings="((Bone=\"woolf_calf_r\",X=Locked,Y=Locked,Z=Free,PreferredAngles=(X=0,Y=0,Z=0)),(Bone=\"woolf_foot_r\",RotationStiffness=0,PositionStiffness=0,X=Locked,MinX=0,MaxX=0,Y=Locked,MinY=0,MaxY=0,Z=Free,MinZ=0,MaxZ=0,bUsePreferredAngles=False,PreferredAngles=(X=0,Y=0 ...(truncated)"
    Debug=(DrawScale=1)
    Effectors=((Bone="woolf_ball_r",Transform=(Rotation=(X=0,Y=0,Z=0,W=1),Translation=(X=0,Y=0,Z=0),Scale3D=(X=1,Y=1,Z=1)),PositionAlpha=1,RotationAlpha=1,StrengthAlpha=1,PinRotation=1,ChainDepth=0))
    ExcludedBones=()
    Root=woolf_thigh_r
    Settings="(Iterations=20,SubIterations=0,MassMultiplier=1,bAllowStretch=False,RootBehavior=PinToInput,PrePullRootSettings=(RotationAlpha=0,RotationAlphaX=1,RotationAlphaY=1,RotationAlphaZ=1,PositionAlpha=1,PositionAlphaX=1,PositionAlphaY=1,PositionAlphaZ=1),GlobalPu ...(truncated)"
    Effectors.0.Transform=&RigUnit_GetTransform(Item=(Type=Bone,Name="foot_r")).Transform
    ExecutePin=@n5.Completed
   )
   > n6 = &RigUnit_SphericalPoseReader(
    ActiveRegionScaleFactors=(PositiveWidth=3,NegativeWidth=4,PositiveHeight=2,NegativeHeight=2.5)
    ActiveRegionSize=0.1
    Debug=(bDrawDebug=false,bDraw2D=false,DebugScale=25,DebugSegments=20,DebugThickness=0.25)
    DriverAxis=(X=1,Y=0,Z=0)
    DriverItem=(Type=Bone,Name="thigh_r")
    FalloffRegionScaleFactors=(PositiveWidth=1,NegativeWidth=1,PositiveHeight=1,NegativeHeight=0.5)
    FalloffSize=0.4
    OptionalParentItem=(Type=Bone,Name="pelvis")
    RotationOffset=(X=-80,Y=0,Z=-20)
   )
   > &RigUnit_ModifyTransforms(
    ItemToModify=((Item=(Type=Bone,Name="woolf_thigh_r"),Transform=(Rotation=(X=0,Y=0,Z=-0.0872,W=0.9962),Translation=(X=0,Y=0,Z=0),Scale3D=(X=1,Y=1,Z=1))))
    WeightMinimum=0
    Weight=@n6.OutputParam
   )
   > &RigUnit_ModifyTransforms(
    ItemToModify=((Item=(Type=Bone,Name="woolf_calf_r"),Transform=(Rotation=(X=0,Y=0,Z=-0.0872,W=0.9962),Translation=(X=0,Y=0,Z=0),Scale3D=(X=1,Y=1,Z=1))))
    WeightMinimum=0
    Weight=@n6.OutputParam
   )
   &RigUnit_ModifyTransforms(
    ExecutePin=()
    ItemToModify=((Item=(Type=Bone,Name="woolf_thigh_l"),Transform=(Rotation=(X=0,Y=0,Z=0.0872,W=0.9962),Translation=(X=0,Y=0,Z=0),Scale3D=(X=1,Y=1,Z=1))))
    Weight=1
    WeightMinimum=0
   )
   > &RigUnit_ModifyTransforms(
    ItemToModify=((Item=(Type=Bone,Name="woolf_thigh_r"),Transform=(Rotation=(X=0,Y=0,Z=0.0872,W=0.9962),Translation=(X=0,Y=0,Z=0),Scale3D=(X=1,Y=1,Z=1))))
    Weight=1
    WeightMinimum=0
   )
   &RigUnit_ParentConstraint(
    AdvancedSettings=(InterpolationType=Average,RotationOrderForFilter=XZY)
    Child=(Type=Bone,Name="woolf_thigh_l")
    ExecutePin=()
    Parents=((Item=(Type=Bone,Name="thigh_l"),Weight=1))
    Weight=1
   )
   > &RigUnit_ParentConstraint(
    AdvancedSettings=(InterpolationType=Average,RotationOrderForFilter=XZY)
    Child=(Type=Bone,Name="woolf_thigh_r")
    Parents=((Item=(Type=Bone,Name="thigh_r"),Weight=1))
    Weight=1
   )
   > &RigUnit_ParentConstraint(
    AdvancedSettings=(InterpolationType=Average,RotationOrderForFilter=XZY)
    Child=(Type=Bone,Name="woolf_calf_l")
    Parents=((Item=(Type=Bone,Name="calf_l"),Weight=1))
    Weight=1
   )
   > &RigUnit_ParentConstraint(
    AdvancedSettings=(InterpolationType=Average,RotationOrderForFilter=XZY)
    Child=(Type=Bone,Name="woolf_calf_r")
    Parents=((Item=(Type=Bone,Name="calf_r"),Weight=1))
    Weight=1
   )
   > &RigUnit_ParentConstraint(
    AdvancedSettings=(InterpolationType=Average,RotationOrderForFilter=XZY)
    Child=(Type=Bone,Name="woolf_ball_l")
    Parents=((Item=(Type=Bone,Name="foot_l"),Weight=1))
    Weight=1
   )
   > &RigUnit_ParentConstraint(
    AdvancedSettings=(InterpolationType=Average,RotationOrderForFilter=XZY)
    Child=(Type=Bone,Name="woolf_ball_r")
    Parents=((Item=(Type=Bone,Name="foot_r"),Weight=1))
    Weight=1
   )
   > &RigUnit_ParentConstraint(
    AdvancedSettings=(InterpolationType=Average,RotationOrderForFilter=XZY)
    Child=(Type=Bone,Name="woolf_toe_l")
    Parents=((Item=(Type=Bone,Name="ball_l"),Weight=1))
    Weight=1
   )
   > &RigUnit_ParentConstraint(
    AdvancedSettings=(InterpolationType=Average,RotationOrderForFilter=XZY)
    Child=(Type=Bone,Name="woolf_toe_r")
    Parents=((Item=(Type=Bone,Name="ball_r"),Weight=1))
    Weight=1
   )
   n7 = &RigUnit_AimBoneMath(
    DebugSettings=(bEnabled=False,Scale=10,WorldOffset=(Rotation=(X=0,Y=0,Z=0,W=1),Translation=(X=0,Y=0,Z=0),Scale3D=(X=1,Y=1,Z=1)))
    Primary=(Weight=1,Axis=(X=1,Y=0,Z=0),Target=(X=1,Y=0,Z=0),Kind=Location,Space=(Type=Bone,Name="None"))
    Secondary=(Weight=1,Axis=(X=0,Y=-1,Z=0),Target=(X=0,Y=0,Z=0),Kind=Location,Space=(Type=Bone,Name="None"))
    Weight=1
    InputTransform=&RigUnit_GetTransform(Item=(Type=Bone,Name="woolf_foot_l"), bInitial=False).Transform
    Secondary.Target=&RigUnit_GetTransform(Item=(Type=Bone,Name="thigh_aim_l"), bInitial=False).Transform.Translation
    Primary.Target=&RigUnit_GetTransform(Item=(Type=Bone,Name="foot_l"), bInitial=False).Transform.Translation
   )
   > &RigUnit_SetTransform(
    Item=(Type=Bone,Name="woolf_foot_l")
    Weight=1
    bInitial=False
    Value=&RigVMFunction_MathTransformMakeAbsolute(Local=(Rotation=(X=0,Y=0,Z=0,W=1),Translation=(X=0,Y=0,Z=0),Scale3D=(X=1,Y=1,Z=1)), Parent=@n7.Result) {template="Make Absolute::Execute(in Local,in Parent,out Global)"}.Global
   )
   n8 = &RigUnit_AimBoneMath(
    DebugSettings=(bEnabled=False,Scale=10,WorldOffset=(Rotation=(X=0,Y=0,Z=0,W=1),Translation=(X=0,Y=0,Z=0),Scale3D=(X=1,Y=1,Z=1)))
    Primary=(Weight=1,Axis=(X=1,Y=0,Z=0),Target=(X=1,Y=0,Z=0),Kind=Location,Space=(Type=Bone,Name="None"))
    Secondary=(Weight=1,Axis=(X=0,Y=-1,Z=0),Target=(X=1,Y=0,Z=0),Kind=Location,Space=(Type=Bone,Name="None"))
    Weight=1
    Secondary.Target=&RigUnit_GetTransform(Item=(Type=Bone,Name="thigh_aim_r"), bInitial=False).Transform.Translation
    InputTransform=&RigUnit_GetTransform(Item=(Type=Bone,Name="woolf_foot_r"), bInitial=False).Transform
    Primary.Target=&RigUnit_GetTransform(Item=(Type=Bone,Name="foot_r"), bInitial=False).Transform.Translation
   )
   > &RigUnit_SetTransform(
    Item=(Type=Bone,Name="woolf_foot_r")
    Weight=1
    bInitial=False
    Value=&RigVMFunction_MathTransformMakeAbsolute(Local=(Rotation=(X=0,Y=0,Z=0,W=1),Translation=(X=0,Y=0,Z=0),Scale3D=(X=1,Y=1,Z=1)), Parent=@n8.Result) {template="Make Absolute::Execute(in Local,in Parent,out Global)"}.Global
   )
   Subtract = /Script/RigVM.RigVMFunction_MathVectorSub(
    B=&RigUnit_GetTransform(Item=(Type=Bone,Name="woolf_thigh_l")).Transform.Translation
    A=&RigUnit_GetTransform(Item=(Type=Bone,Name="woolf_foot_l")).Transform.Translation
   ) {is=struct RigVMUnitNode, template="Subtract::Execute(in A,in B,out Result)"}
   Unit = /Script/RigVM.RigVMFunction_MathVectorUnit(Value=@Subtract.Result) {is=struct RigVMUnitNode, template="Unit::Execute(in Value,out Result)"}
   RigVMFunction_MathQuaternionFromAxisAndAngle = /Script/RigVM.RigVMFunction_MathQuaternionFromAxisAndAngle(Angle=-1, Axis=@Unit.Result) {is=struct RigVMUnitNode, template="FromAxisAndAngle::Execute(in Axis,in Angle,out Result)"}
   Set Transform = &RigUnit_SetRotation(
    ExecutePin=()
    Item=(Type=Bone,Name="woolf_thigh_l")
    Space=GlobalSpace
    Weight=1
    bInitial=false
    Value=/Script/RigVM.RigVMFunction_MathQuaternionMul(
     B=&RigUnit_GetTransform(Item=(Type=Bone,Name="woolf_thigh_l"), bInitial=false).Transform.Rotation
     A=@RigVMFunction_MathQuaternionFromAxisAndAngle.Result
    ) {is=struct RigVMUnitNode, template="Multiply::Execute(in A,in B,out Result)"}.Result
   )
  ]}
 New Function_ContainedGraph {nodes [
   n1 = RigVMFunctionEntryNode()
   RigVMFunction_Sequence = &RigVMFunction_Sequence(ExecuteContext=@n1.ExecuteContext)
   &RigUnit_AimConstraintLocalSpaceOffset(
    Child=(Type=Bone,Name="toe_outter_l")
    Filter=(bX=True,bY=True,bZ=True)
    Parents=((Item=(Type=Control,Name="ball_l_ctrl"),Weight=1))
    Weight=1
    WorldUp=(Target=(X=0,Y=0,Z=1),Kind=Location,Space=(Type=Control,Name="ball_l_ctrl"))
    ExecutePin=@RigVMFunction_Sequence.A
   )
   > &RigUnit_AimConstraintLocalSpaceOffset(
    Child=(Type=Bone,Name="toe_mid_l")
    Filter=(bX=True,bY=True,bZ=True)
    Parents=((Item=(Type=Control,Name="ball_l_ctrl"),Weight=1))
    Weight=1
    WorldUp=(Target=(X=0,Y=0,Z=1),Kind=Location,Space=(Type=Control,Name="ball_l_ctrl"))
   )
   > &RigUnit_AimConstraintLocalSpaceOffset(
    Child=(Type=Bone,Name="toe_inner_l")
    Filter=(bX=True,bY=True,bZ=True)
    Parents=((Item=(Type=Control,Name="ball_l_ctrl"),Weight=1))
    Weight=1
    WorldUp=(Target=(X=0,Y=0,Z=1),Kind=Location,Space=(Type=Control,Name="ball_l_ctrl"))
   )
   > &RigUnit_AimConstraintLocalSpaceOffset(
    Child=(Type=Bone,Name="toe_outter_r")
    Filter=(bX=True,bY=True,bZ=True)
    Parents=((Item=(Type=Control,Name="ball_r_ctrl"),Weight=1))
    Weight=1
    WorldUp=(Target=(X=0,Y=0,Z=1),Kind=Location,Space=(Type=Control,Name="ball_r_ctrl"))
   )
   > &RigUnit_AimConstraintLocalSpaceOffset(
    Child=(Type=Bone,Name="toe_mid_r")
    Filter=(bX=True,bY=True,bZ=True)
    Parents=((Item=(Type=Control,Name="ball_r_ctrl"),Weight=1))
    Weight=1
    WorldUp=(Target=(X=0,Y=0,Z=1),Kind=Location,Space=(Type=Control,Name="ball_r_ctrl"))
   )
   > &RigUnit_AimConstraintLocalSpaceOffset(
    Child=(Type=Bone,Name="toe_inner_r")
    Filter=(bX=True,bY=True,bZ=True)
    Parents=((Item=(Type=Control,Name="ball_r_ctrl"),Weight=1))
    Weight=1
    WorldUp=(Target=(X=0,Y=0,Z=1),Kind=Location,Space=(Type=Control,Name="ball_r_ctrl"))
   )
   > Get Delta Ref = RigVMFunctionReferenceNode(Item=(Type=Bone,Name="foot_l"))
   > Get Delta Ref_1 = RigVMFunctionReferenceNode(Item=(Type=Bone,Name="foot_r"))
   Remap = &RigVMFunction_MathDoubleRemap(SourceMaximum=50, SourceMinimum=0, TargetMaximum=1, TargetMinimum=0, Value=@"Get Delta Ref".Result.Rotation.Yaw)
   > Modify Transforms with Mirror = RigVMFunctionReferenceNode(
    Base Modify Data="((Item=(Type=Bone,Name=\"toe_inner_l\"),Transform=(Rotation=(X=0,Y=0,Z=-0.1305,W=0.9914),Translation=(X=-0.6923,Y=0,Z=-1.3044),Scale3D=(X=1,Y=1,Z=1))),(Item=(Type=Bone,Name=\"toe_outter_l\"),Transform=(Rotation=(X=0,Y=0,Z=0.2164,W=0.9763),Translation=(X=0,Y=0, ...(truncated)"
    Mirror Modify Data="((Item=(Type=Bone,Name=\"toe_inner_r\"),Transform=(Rotation=(X=0,Y=0,Z=0,W=1),Translation=(X=0,Y=0,Z=0),Scale3D=(X=1,Y=1,Z=1))),(Item=(Type=Bone,Name=\"toe_outter_r\"),Transform=(Rotation=(X=0,Y=0,Z=0,W=1),Translation=(X=0,Y=0,Z=0),Scale3D=(X=1,Y=1,Z=1))),(Ite ...(truncated)"
    Weight=@Remap.Result
    "Mirror Weight "=&RigVMFunction_MathDoubleRemap(SourceMaximum=50, SourceMinimum=0, TargetMaximum=1, TargetMinimum=0, Value=@"Get Delta Ref_1".Result.Rotation.Yaw).Result
   )
   > Get Delta Ref_2 = RigVMFunctionReferenceNode(Item=(Type=Bone,Name="foot_l"))
   > Get Delta Ref_3 = RigVMFunctionReferenceNode(Item=(Type=Bone,Name="foot_r"))
   > Modify Transforms with Mirror_1 = RigVMFunctionReferenceNode(
    Base Modify Data="((Item=(Type=Bone,Name=\"toe_inner_l\"),Transform=(Rotation=(X=0,Y=0,Z=0.0872,W=0.9962),Translation=(X=0,Y=0,Z=0),Scale3D=(X=1,Y=1,Z=1))),(Item=(Type=Bone,Name=\"toe_outter_l\"),Transform=(Rotation=(X=0,Y=0,Z=-0.0872,W=0.9962),Translation=(X=0,Y=0,Z=0),Scale3D ...(truncated)"
    Mirror Modify Data=((Item=(Type=Bone,Name="toe_inner_r"),Transform=(Rotation=(X=0,Y=0,Z=0,W=1),Translation=(X=0,Y=0,Z=0),Scale3D=(X=1,Y=1,Z=1))),(Item=(Type=Bone,Name="toe_outter_r"),Transform=(Rotation=(X=0,Y=0,Z=0,W=1),Translation=(X=0,Y=0,Z=0),Scale3D=(X=1,Y=1,Z=1))))
    Weight=&RigVMFunction_MathDoubleRemap(SourceMaximum=-20, SourceMinimum=0, TargetMaximum=1, TargetMinimum=0, Value=@"Get Delta Ref_2".Result.Rotation.Yaw).Result
    "Mirror Weight "=&RigVMFunction_MathDoubleRemap(SourceMaximum=-20, SourceMinimum=0, TargetMaximum=1, TargetMinimum=0, Value=@"Get Delta Ref_3".Result.Rotation.Yaw).Result
   )
   RigVMFunctionReturnNode(ExecuteContext=@RigVMFunction_Sequence.B)
  ]}
 RigVMFunctionLibrary {nodes [
   Leg Setup = RigVMCollapseNode(ExecuteContext=())
   ToeSetup = RigVMCollapseNode(ExecuteContext=())
  ]}
 RigVMModel {nodes [
   RigUnit_BeginExecution = &RigUnit_BeginExecution()
   RigVMFunction_Sequence = &RigVMFunction_Sequence(ExecuteContext=@RigUnit_BeginExecution.ExecutePin)
   FN_DeformRig_FNC = RigVMFunctionReferenceNode(
    DeformRig_ Curve="(deformRig_l_upperarm_up_alpha_72_#1=False,deformRig_r_upperarm_up_alpha_73_#2=False,deformRig_l_upperarm_fr_alpha_74_#3=False,deformRig_r_upperarm_fr_alpha_75_CB51B8 ...(truncated)"
    DeformRig_Bool="(upperarm_1_#4=True,elbow_3_#5=True,wrist_7_#6=True,groin_8_#7=True,glute_17_#8=True,knee_9_40C408324CA261 ...(truncated)"
    DeformRig_Debug="(upperarm_debug=False,wrist_debug=False,lat_debug=False,groinGlute_debug=False,upperarm_debug_18_#4=False,wrist_debug_20_#6=False,groinGlute_debug_27_#7=False,lat_deb ...(truncated)"
    DeformRig_Value="(upperarm_up_18_#9=(Translation=(X=0,Y=0,Z=8),Scale3D=(X=0.7,Y=1,Z=1)),upperarm_fwd_fr_19_#10=(Translation=(X=0,Y=10,Z=0),Scale3D=(X=0.5,Y=1,Z=1)),upperarm_fwd_bk_20_9973AAD84EE0A1E349BC9E85067D4A3 ...(truncated)"
    ExecuteContext=@RigVMFunction_Sequence.A
   )
   Leg Setup = RigVMFunctionReferenceNode(ExecuteContext=@RigVMFunction_Sequence.B)
   > ToeSetup = RigVMFunctionReferenceNode()
  ]}
 SequenceAggregate_1_ContainedGraph {nodes [
   n1 = RigVMFunctionEntryNode()
   SequenceAggregate = &RigVMFunction_Sequence(ExecuteContext=@n1.ExecuteContext)
   SequenceAggregate_1 = &RigVMFunction_Sequence(ExecuteContext=@SequenceAggregate.B)
   RigVMFunctionReturnNode(A=@SequenceAggregate.A, B=@SequenceAggregate_1.A, C=@SequenceAggregate_1.B)
  ]}
]
hierarchy [
 root {loc=<0 0 0>, rot=<0 0 0 1>}
 pelvis {parent=root, loc=<0 -0.7321 92.4218>, rot=<0.0127 0.707 -0.0127 -0.707>}
 spine_01 {parent=pelvis, loc=<5.6002 -0.5128 0>, rot=<0 0 0.034 -0.9994>}
 spine_02 {parent=spine_01, loc=<8.0756 0.0139 0>, rot=<0 0 0.0238 -0.9997>}
 spine_03 {parent=spine_02, loc=<7.0395 0 0>, rot=<0 0 0.0149 -0.9999>}
 spine_04 {parent=spine_03, loc=<8.1023 0 0>, rot=<0 0 -0.0132 -0.9999>}
 spine_05 {parent=spine_04, loc=<10.0141 0 0>, rot=<0 0 -0.0015 -1>}
 clavicle_l {parent=spine_05, loc=<2.5856 -0.2866 -4.3777>, rot=<0.0465 -0.7289 -0.009 -0.6829>}
 upperarm_l {parent=clavicle_l, loc=<18.2264 0.1342 -0.001>, rot=<0.0253 -0.2693 0.0013 -0.9627>}
 lowerarm_l {parent=upperarm_l, loc=<29.0071 0.1686 -1.239>, rot=<0.0391 -0.0099 -0.2453 -0.9686>}
 hand_l {parent=lowerarm_l, loc=<25.8607 0.222 1.0274>, rot=<-0.7193 -0.0988 0.1193 -0.6773>}
 thumb_01_l {parent=hand_l, loc=<1.3118 -0.3191 -2.459>, rot=<0.6546 -0.4042 0.0664 -0.6354>}
 thumb_02_l {parent=thumb_01_l, loc=<5.9336 -0.0448 0.0401>, rot=<-0.0037 0.0213 0.298 -0.9543>}
 thumb_03_l {parent=thumb_02_l, loc=<2.2223 0.3682 -0.0907>, rot=<0.0118 -0.013 0.0066 -0.9998>}
 index_metacarpal_l {parent=hand_l, loc=<1.6535 0.0183 -2.96>, rot=<0.0158 -0.0124 -0.0173 -0.9996>}
 index_01_l {parent=index_metacarpal_l, loc=<8.3008 0.0357 -0.0524>, rot=<-0.0044 0.0485 0.0924 -0.9945>}
 index_02_l {parent=index_01_l, loc=<3.3587 -0.3004 0.0452>, rot=<-0.002 0.0027 0.3225 -0.9466>}
 index_03_l {parent=index_02_l, loc=<2.5578 -0.123 -0.0252>, rot=<-0.002 0.0067 0.2558 -0.9667>}
 middle_metacarpal_l {parent=hand_l, loc=<2.1856 -0.1725 -0.5984>, rot=<0.006 0.0168 -0.0329 -0.9993>}
 middle_01_l {parent=middle_metacarpal_l, loc=<7.3275 -0.2447 0.0129>, rot=<-0.0208 0.0042 0.0858 -0.9961>}
 middle_02_l {parent=middle_01_l, loc=<4.0123 0.1401 0.1175>, rot=<-0.0218 0.0258 0.3512 -0.9357>}
 middle_03_l {parent=middle_02_l, loc=<2.6077 0.0534 0.1373>, rot=<0.0082 0.0188 0.1708 -0.9851>}
 ring_metacarpal_l {parent=hand_l, loc=<2.2074 -0.3625 1.3638>, rot=<-0.0284 0.0722 -0.0458 -0.9959>}
 ring_01_l {parent=ring_metacarpal_l, loc=<7.0295 -0.6197 -0.1446>, rot=<-0.0489 -0.0252 0.1892 -0.9804>}
 ring_02_l {parent=ring_01_l, loc=<3.6174 0.4647 -0.1027>, rot=<0 0 0.2806 -0.9598>}
 ring_03_l {parent=ring_02_l, loc=<2.4152 0.114 0.0199>, rot=<0 0 0.1189 -0.9929>}
 pinky_metacarpal_l {parent=hand_l, loc=<1.7616 -0.9753 3.3814>, rot=<-0.0397 0.0814 -0.069 -0.9935>}
 pinky_01_l {parent=pinky_metacarpal_l, loc=<7.1709 -0.4831 -0.1094>, rot=<-0.0654 -0.0035 0.2242 -0.9723>}
 pinky_02_l {parent=pinky_01_l, loc=<2.8274 0.0893 -0.1043>, rot=<0.0037 0.0011 0.2854 -0.9584>}
 pinky_03_l {parent=pinky_02_l, loc=<2.2046 0.0307 -0.0078>, rot=<-0.0002 0.0098 0.1445 -0.9895>}
 deform_wrist_dn_base_l {parent=hand_l, loc=<0 -3.4173 0>, rot=<0 0 0 -1>}
 deform_wrist_dn_l {parent=deform_wrist_dn_base_l, loc=<0 0 0>, rot=<0 0 0 1>}
 deform_wrist_up_base_l {parent=hand_l, loc=<0 2.7805 0>, rot=<0 0 0 -1>}
 deform_wrist_up_l {parent=deform_wrist_up_base_l, loc=<0 0 0>, rot=<0 0 0 1>}
 weapon_l {parent=hand_l, loc=<-0.1974 -3.4111 0.3811>, rot=<0 0 0 1>}
 hand_attach_l {parent=hand_l, loc=<8.5 -3 0>, rot=<0.7071 0.7071 0 0>}
 lowerarm_twist_02_l {parent=lowerarm_l, loc=<9.6199 0 0>, rot=<0 0 0 -1>}
 lowerarm_twist_01_l {parent=lowerarm_l, loc=<17.8549 0 0>, rot=<0 0 0 -1>}
 deform_elbow_in_base_l {parent=lowerarm_l, loc=<0.2112 5.645 0.8274>, rot=<0 0 0 -1>}
 deform_elbow_in_l {parent=deform_elbow_in_base_l, loc=<0 0 0>, rot=<0 0 0 1>}
 deform_elbow_out_base_l {parent=lowerarm_l, loc=<0 -7.5129 0>, rot=<0 0 0 -1>}
 deform_elbow_out_l {parent=deform_elbow_out_base_l, loc=<0 0 0>, rot=<0 0 0 1>}
 upperarm_twist_01_l {parent=upperarm_l, loc=<7.9796 0.0843 -0.6195>, rot=<0 0 0 -1>}
 deform_upperarm_bk_base_l {parent=upperarm_twist_01_l, loc=<-6.8983 -13.3826 -1.3258>, rot=<0 0 0 1>}
 deform_upperarm_bk_l {parent=deform_upperarm_bk_base_l, loc=<0 0 0>, rot=<0 0 0 1>}
 deform_upperarm_fr_base_l {parent=upperarm_twist_01_l, loc=<-8.3487 10.7117 -2.1455>, rot=<0 0 0 1>}
 deform_upperarm_fr_l {parent=deform_upperarm_fr_base_l, loc=<0 0 0>, rot=<0 0 0 1>}
 deform_upperarm_up_base_l {parent=upperarm_twist_01_l, loc=<-11.2012 -1.2234 8.7784>, rot=<0 0 0 1>}
 deform_upperarm_up_l {parent=deform_upperarm_up_base_l, loc=<0 0 0>, rot=<0 0 0 1>}
 upperarm_twist_02_l {parent=upperarm_l, loc=<19.5059 0.0843 -0.6195>, rot=<0 0 0 -1>}
 deform_bicep_base_l {parent=upperarm_twist_02_l, loc=<0 8.0494 0>, rot=<0 0 0 1>}
 deform_bicep_l {parent=deform_bicep_base_l, loc=<0 0 0>, rot=<0 0 0 1>}
 neck_01 {parent=spine_05, loc=<10.815 1.9196 0>, rot=<0 0 -0.1864 -0.9825>}
 neck_02 {parent=neck_01, loc=<4.7347 0.3045 0>, rot=<0 0 0.0363 -0.9993>}
 head {parent=neck_02, loc=<5.7632 0.5552 0>, rot=<0 0 0.115 -0.9934>}
 clavicle_r {parent=spine_05, loc=<2.5855 -0.2866 4.3777>, rot=<0.7289 0.0465 -0.6829 0.009>}
 upperarm_r {parent=clavicle_r, loc=<-18.2264 -0.1342 0.001>, rot=<0.0253 -0.2693 0.0013 -0.9627>}
 lowerarm_r {parent=upperarm_r, loc=<-29.0071 -0.1686 1.239>, rot=<0.0391 -0.0099 -0.2453 -0.9686>}
 hand_r {parent=lowerarm_r, loc=<-25.8607 -0.222 -1.0274>, rot=<-0.7193 -0.0988 0.1193 -0.6773>}
 thumb_01_r {parent=hand_r, loc=<-1.3118 0.3191 2.459>, rot=<0.6546 -0.4042 0.0664 -0.6354>}
 thumb_02_r {parent=thumb_01_r, loc=<-5.9336 0.0448 -0.0401>, rot=<-0.0037 0.0213 0.298 -0.9543>}
 thumb_03_r {parent=thumb_02_r, loc=<-2.2223 -0.3682 0.0907>, rot=<0.0118 -0.013 0.0066 -0.9998>}
 index_metacarpal_r {parent=hand_r, loc=<-1.6535 -0.0183 2.96>, rot=<0.0158 -0.0124 -0.0173 -0.9996>}
 index_01_r {parent=index_metacarpal_r, loc=<-8.3008 -0.0357 0.0524>, rot=<-0.0044 0.0485 0.0924 -0.9945>}
 index_02_r {parent=index_01_r, loc=<-3.3587 0.3004 -0.0452>, rot=<-0.002 0.0027 0.3225 -0.9466>}
 index_03_r {parent=index_02_r, loc=<-2.5578 0.123 0.0252>, rot=<-0.002 0.0067 0.2558 -0.9667>}
 middle_metacarpal_r {parent=hand_r, loc=<-2.1856 0.1725 0.5984>, rot=<0.006 0.0168 -0.0329 -0.9993>}
 middle_01_r {parent=middle_metacarpal_r, loc=<-7.3275 0.2447 -0.0129>, rot=<-0.0208 0.0042 0.0858 -0.9961>}
 middle_02_r {parent=middle_01_r, loc=<-4.0123 -0.1401 -0.1175>, rot=<-0.0218 0.0258 0.3512 -0.9357>}
 middle_03_r {parent=middle_02_r, loc=<-2.6077 -0.0534 -0.1373>, rot=<0.0082 0.0188 0.1708 -0.9851>}
 ring_metacarpal_r {parent=hand_r, loc=<-2.2074 0.3625 -1.3638>, rot=<-0.0284 0.0722 -0.0458 -0.9959>}
 ring_01_r {parent=ring_metacarpal_r, loc=<-7.0295 0.6197 0.1446>, rot=<-0.0489 -0.0252 0.1892 -0.9804>}
 ring_02_r {parent=ring_01_r, loc=<-3.6174 -0.4647 0.1027>, rot=<0 0 0.2806 -0.9598>}
 ring_03_r {parent=ring_02_r, loc=<-2.4152 -0.114 -0.0199>, rot=<0 0 0.1189 -0.9929>}
 pinky_metacarpal_r {parent=hand_r, loc=<-1.7616 0.9753 -3.3814>, rot=<-0.0397 0.0814 -0.069 -0.9935>}
 pinky_01_r {parent=pinky_metacarpal_r, loc=<-7.1709 0.4831 0.1094>, rot=<-0.0654 -0.0035 0.2242 -0.9723>}
 pinky_02_r {parent=pinky_01_r, loc=<-2.8274 -0.0893 0.1043>, rot=<0.0037 0.0011 0.2854 -0.9584>}
 pinky_03_r {parent=pinky_02_r, loc=<-2.2046 -0.0307 0.0078>, rot=<-0.0002 0.0098 0.1445 -0.9895>}
 deform_wrist_dn_base_r {parent=hand_r, loc=<0 3.4172 0>, rot=<0 0 0 -1>}
 deform_wrist_dn_r {parent=deform_wrist_dn_base_r, loc=<0 0 0>, rot=<0 0 0 -1>}
 deform_wrist_up_base_r {parent=hand_r, loc=<-0.0001 -2.7804 0>, rot=<0 0 0 -1>}
 deform_wrist_up_r {parent=deform_wrist_up_base_r, loc=<0 0 0>, rot=<0 0 0 -1>}
 hand_attach_r {parent=hand_r, loc=<-8.5 3 0>, rot=<0 0 -0.7071 -0.7071>}
 weapon_r {parent=hand_r, loc=<0.1974 3.4111 -0.3811>, rot=<0 0 0 1>}
 lowerarm_twist_02_r {parent=lowerarm_r, loc=<-9.6199 0 0>, rot=<0 0 0 -1>}
 lowerarm_twist_01_r {parent=lowerarm_r, loc=<-17.8549 0 0>, rot=<0 0 0 -1>}
 deform_elbow_in_base_r {parent=lowerarm_r, loc=<-0.2114 -5.6449 -0.827>, rot=<0 0 0 -1>}
 deform_elbow_in_r {parent=deform_elbow_in_base_r, loc=<0 0 0>, rot=<0 0 0 1>}
 deform_elbow_out_base_r {parent=lowerarm_r, loc=<-0.0001 7.5129 0.0002>, rot=<0 0 0 -1>}
 deform_elbow_out_r {parent=deform_elbow_out_base_r, loc=<0 0 0>, rot=<0 0 0 1>}
 upperarm_twist_01_r {parent=upperarm_r, loc=<-7.9796 -0.0843 0.6195>, rot=<0 0 0 -1>}
 deform_upperarm_bk_base_r {parent=upperarm_twist_01_r, loc=<6.8985 13.3826 1.3257>, rot=<0 0 0 1>}
 deform_upperarm_bk_r {parent=deform_upperarm_bk_base_r, loc=<0 0 0>, rot=<0 0 0 1>}
 deform_upperarm_fr_base_r {parent=upperarm_twist_01_r, loc=<8.3485 -10.7117 2.1456>, rot=<0 0 0 1>}
 deform_upperarm_fr_r {parent=deform_upperarm_fr_base_r, loc=<0 0 0>, rot=<0 0 0 1>}
 deform_upperarm_up_base_r {parent=upperarm_twist_01_r, loc=<11.2015 1.2234 -8.7788>, rot=<0 0 0 1>}
 deform_upperarm_up_r {parent=deform_upperarm_up_base_r, loc=<0 0 0>, rot=<0 0 0 1>}
 upperarm_twist_02_r {parent=upperarm_r, loc=<-19.5059 -0.0843 0.6195>, rot=<0 0 0 -1>}
 deform_bicep_base_r {parent=upperarm_twist_02_r, loc=<0.0002 -8.0494 -0.0003>, rot=<0 0 0 1>}
 deform_bicep_r {parent=deform_bicep_base_r, loc=<0 0 0>, rot=<0 0 0 1>}
 attach_cape {parent=spine_05, loc=<9.017 -11.714 0>, rot=<0.0616 -0.7044 -0.0616 -0.7044>}
 cape {parent=attach_cape, type=Null, loc=<0 0 0>, rot=<0 0 0 1>}
 attach_backpack {parent=spine_05, loc=<-0.123 -13.473 0>, rot=<-0.0093 -0.707 0.0093 -0.707>}
 Backpack_BR {parent=attach_backpack, type=Null, loc=<0 0 0>, rot=<0 0 0 1>}
 deform_pec_aim_base_l {parent=spine_04, loc=<5.0227 13.6719 -2.6508>, rot=<0.0672 -0.5201 0.3228 -0.7879>}
 deform_pec_aim_l {parent=deform_pec_aim_base_l, loc=<0 0 0>, rot=<0 0 0 1>}
 deform_pec_base_l {parent=deform_pec_aim_l, loc=<16.3279 0 0>, rot=<0 0 0 1>}
 deform_pec_l {parent=deform_pec_base_l, loc=<0 0 0>, rot=<0 0 0 1>}
 deform_pec_aim_base_r {parent=spine_04, loc=<5.0225 13.6718 2.6508>, rot=<-0.5201 -0.0672 0.7879 0.3228>}
 deform_pec_aim_r {parent=deform_pec_aim_base_r, loc=<0 0 0>, rot=<0 0 0 1>}
 deform_pec_base_r {parent=deform_pec_aim_r, loc=<-16.3279 0 0>, rot=<0 0 0 -1>}
 deform_pec_r {parent=deform_pec_base_r, loc=<0 0 0.0002>, rot=<0 0 0 -1>}
 deform_lat_l_base {parent=spine_03, loc=<2.5561 -10.3716 -17.6918>, rot=<0 0 0 -1>}
 deform_lat_l {parent=deform_lat_l_base, loc=<0 0 0>, rot=<0 0 0 1>}
 deform_lat_r_base {parent=spine_03, loc=<2.5562 -10.3716 17.6918>, rot=<0 0 1 0>}
 deform_lat_r {parent=deform_lat_r_base, loc=<0 0 0>, rot=<0 0 0 -1>}
 thigh_r {parent=pelvis, loc=<-1.8716 -0.0109 11.9936>, rot=<0.0572 -0.0012 0.9982 0.0155>}
 calf_r {parent=thigh_r, loc=<41.9178 0.2246 -0.0926>, rot=<0.0006 -0.0161 -0.0391 -0.9991>}
 foot_r {parent=calf_r, loc=<38.7057 -1.2438 -0.2452>, rot=<0.0975 -0.0474 0.0317 -0.9936>}
 ball_r {parent=foot_r, loc=<6.1614 -13.1469 -0.6807>, rot=<0.0001 -0.0001 -0.7063 -0.7079>}
 woolf_toe_r {parent=ball_r, loc=<4.4558 -0.0766 0.0538>, rot=<0.0022 0.1012 0.0042 -0.9949>}
 toe_inner_r {parent=woolf_toe_r, loc=<-3.0192 -1.1578 -5.1018>, rot=<-0.7054 -0.0485 -0.0485 -0.7054>}
 toe_mid_r {parent=woolf_toe_r, loc=<-3.0509 -1.0245 0.6387>, rot=<-0.7071 -0.0829 0.0029 -0.7022>}
 toe_outter_r {parent=woolf_toe_r, loc=<-0.3077 -2.1186 7.0597>, rot=<-0.705 -0.098 0.0543 -0.7003>}
 ball_r_ctrl {controlType=EULER_TRANSFORM, parent=ball_r, type=Control, loc=<0 0 0>, rot=<0 0 0 1>}
 woolf_ball_r {parent=foot_r, loc=<6.3114 -2.591 0.6257>, rot=<-0.0699 0.0732 -0.6997 -0.7072>}
 toe_thumb_r {parent=woolf_ball_r, loc=<-1.2999 -1.3622 -7.1698>, rot=<-0.5596 0.3259 -0.4323 -0.6275>}
 calf_twist_02_r {parent=calf_r, loc=<15.9268 0.0967 -0.0397>, rot=<0.0237 0.0095 -0.0035 -0.9997>}
 calf_twist_01_r {parent=calf_r, loc=<32.9334 0.0967 -0.0397>, rot=<0.0237 0.0095 -0.0035 -0.9997>}
 deform_knee_in_base_r {parent=calf_r, loc=<-1.2981 8.0356 0.0496>, rot=<0 0 0 -1>}
 deform_knee_in_r {parent=deform_knee_in_base_r, loc=<0 0 0>, rot=<0 0 0 1>}
 deform_knee_out_base_r {parent=calf_r, loc=<0 -6.8739 0>, rot=<0 0 0 -1>}
 deform_knee_out_r {parent=deform_knee_out_base_r, loc=<0 0 0>, rot=<0 0 0 1>}
 thigh_aim_r {parent=calf_r, loc=<-3.3575 -43.705 -0.0215>, rot=<-0.0006 0.0161 0.0391 -0.9991>}
 thigh_twist_01_r {parent=thigh_r, loc=<21.6303 0 0>, rot=<0.0013 0 0 -1>}
 deform_groin_base_r {parent=thigh_r, loc=<2.1908 -8.5661 1.7442>, rot=<0 0 0 -1>}
 deform_groin_r {parent=deform_groin_base_r, loc=<0 0 0>, rot=<0 0 0 -1>}
 thigh_l {parent=pelvis, loc=<-1.8716 -0.0109 -11.9936>, rot=<0.0012 0.0572 0.0155 -0.9982>}
 calf_l {parent=thigh_l, loc=<-41.9178 -0.2247 0.0926>, rot=<0.0006 -0.0161 -0.0391 -0.9991>}
 foot_l {parent=calf_l, loc=<-38.7057 1.2439 0.2452>, rot=<0.0975 -0.0474 0.0317 -0.9936>}
 ball_l {parent=foot_l, loc=<-6.1614 13.1469 0.6808>, rot=<0.0001 -0.0001 -0.7063 -0.7079>}
 woolf_toe_l {parent=ball_l, loc=<-4.4558 0.0766 -0.0538>, rot=<0.0022 0.1012 0.0042 -0.9949>}
 toe_outter_l {parent=woolf_toe_l, loc=<0.3077 2.1186 -7.0597>, rot=<-0.705 -0.098 0.0543 -0.7003>}
 toe_mid_l {parent=woolf_toe_l, loc=<3.0509 1.0245 -0.6387>, rot=<-0.7071 -0.0829 0.0029 -0.7022>}
 toe_inner_l {parent=woolf_toe_l, loc=<3.0192 1.1578 5.1018>, rot=<-0.7054 -0.0485 -0.0485 -0.7054>}
 ball_l_ctrl {controlType=EULER_TRANSFORM, parent=ball_l, type=Control, loc=<0 0 0>, rot=<0 0 0 1>}
 woolf_ball_l {parent=foot_l, loc=<-6.3114 2.591 -0.6257>, rot=<-0.0699 0.0732 -0.6997 -0.7072>}
 toe_thumb_l {parent=woolf_ball_l, loc=<1.2999 1.3622 7.1698>, rot=<-0.5596 0.3259 -0.4323 -0.6275>}
 calf_twist_02_l {parent=calf_l, loc=<-15.9268 -0.0967 0.0397>, rot=<0.0237 0.0095 -0.0035 -0.9997>}
 calf_twist_01_l {parent=calf_l, loc=<-32.9335 -0.0967 0.0397>, rot=<0.0237 0.0095 -0.0035 -0.9997>}
 deform_knee_in_base_l {parent=calf_l, loc=<1.298 -8.0356 -0.0496>, rot=<0 0 0 -1>}
 deform_knee_in_l {parent=deform_knee_in_base_l, loc=<0 0 0>, rot=<0 0 0 1>}
 deform_knee_out_base_l {parent=calf_l, loc=<0 6.8739 0>, rot=<0 0 0 -1>}
 deform_knee_out_l {parent=deform_knee_out_base_l, loc=<0 0 0>, rot=<0 0 0 1>}
 thigh_aim_l {parent=calf_l, loc=<3.3575 43.705 0.0215>, rot=<-0.0006 0.0161 0.0391 -0.9991>}
 thigh_twist_01_l {parent=thigh_l, loc=<-21.6303 0 0>, rot=<0.0013 0 0 -1>}
 deform_groin_base_l {parent=thigh_l, loc=<-2.1907 8.5661 -1.7442>, rot=<0 0 0 -1>}
 deform_groin_l {parent=deform_groin_base_l, loc=<0 0 0>, rot=<0 0 0 1>}
 deform_glute_base_l {parent=pelvis, loc=<-10.1877 -11.6217 -12.899>, rot=<0.0275 0.0501 -0.4472 -0.8926>}
 deform_glute_l {parent=deform_glute_base_l, loc=<0 0 0>, rot=<0 0 0 1>}
 deform_glute_base_r {parent=pelvis, loc=<-10.1877 -11.6218 12.899>, rot=<0.0501 -0.0275 0.8926 -0.4472>}
 deform_glute_r {parent=deform_glute_base_r, loc=<0 0 0>, rot=<0 0 0 1>}
 woolf_thigh_l {parent=pelvis, loc=<-1.8716 -0.0109 -11.9936>, rot=<0.0135 0.0542 0.0163 -0.9983>}
 woolf_calf_l {parent=woolf_thigh_l, loc=<-39.5635 0.86 0.0143>, rot=<-0.0084 -0.0079 -0.0395 -0.9992>}
 woolf_foot_l {parent=woolf_calf_l, loc=<-26.9854 -9.649 -0.1016>, rot=<0.0008 -0.0019 0.9605 0.2784>}
 woolf_thigh_r {parent=pelvis, loc=<-1.8716 -0.0109 11.9936>, rot=<0.0542 -0.0135 0.9983 0.0163>}
 woolf_calf_r {parent=woolf_thigh_r, loc=<39.5635 -0.86 -0.0144>, rot=<-0.0084 -0.0079 -0.0395 -0.9992>}
 woolf_foot_r {parent=woolf_calf_r, loc=<26.9854 9.649 0.1017>, rot=<-0.0019 -0.0008 0.2784 -0.9605>}
]
