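# Animation Blueprint for the large male BentBaton PikeAway character body part, adding custom toe posing to the player's animation.
# It copies the attached parent mesh's pose through a Control Rig, then blends slot animation onto toe bones using the CustomAnim_Override curve.
# A toe-curl pose then layers onto each foot's woolf bones.
# The owner anim instance's toe-splay curves drive those weights, which are zeroed while idle or settling.
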AnimBlueprint /BRCosmetics/Characters/Player/Male/Large/Bodies/M_LRG_BentBaton_PikeAway/Meshes/M_LRG_BentBaton_PikeAway_AnimBP
target_skeleton=&/M_LRG_BentBaton_PikeAway_Skeleton
graphs [
 AnimGraph {nodes [
   Layered blend per bone(BlendWeights_0=Get Is Playing Custom Anim() -> (Is Playing Custom Anim:double)) {layerSetup [{branchFilters [
       {boneName=thigh_l}
       {boneName=thigh_r}
       {boneName=calf_l}
       {boneName=calf_r}
       {boneName=foot_l}
       {boneName=foot_r}
       {boneName=ball_l}
       {boneName=ball_r}
      ]}]}
   n1 = Control Rig(Source=Copy Pose From Mesh() {bUseAttachedParent=true}) {
    alphaBoolBlend {blendOption=Linear}
    controlRigAssetReference {blueprintRigClass=&/M_LRG_BentBaton_PikeAway_CtrlRig.M_LRG_BentBaton_PikeAway_CtrlRig_C}
   }
   n2 = Layered blend per bone(BasePose=@n1, BlendPoses_0=Slot 'DefaultSlot'(), BlendWeights_0=Get Is Playing Custom Anim()) {layerSetup [{branchFilters [
       {boneName=toe_inner_l}
       {boneName=toe_mid_l}
       {boneName=toe_outter_l}
       {boneName=toe_thumb_l}
       {boneName=toe_inner_r}
       {boneName=toe_mid_r}
       {boneName=toe_outter_r}
       {boneName=toe_thumb_r}
      ]}]}
   n3 = Sequence Player 'BeastGirl_ToeCurl_Pose'() {bLoopAnimation=false, sequence=&/BeastGirl_ToeCurl_Pose}
   n4 = Get ToeCurl_Left() -> (ToeCurl_Left:double)
   n5 = Sequence Player 'BeastGirl_ToeCurl_Pose'() {bLoopAnimation=false, sequence=&BeastGirl_ToeCurl_Pose}
   Output Pose(Result=Layered blend per bone(
     BasePose=Layered blend per bone(BasePose=@n2, BlendPoses_0=@n3, BlendWeights_0=@n4) {layerSetup [{branchFilters [{boneName=woolf_toe_l}, {boneName=woolf_ball_l}]}]}
     BlendPoses_0=@n5
     BlendWeights_0=Get ToeCurl_Right() -> (ToeCurl_Right:double)
    ) {layerSetup [{branchFilters [{boneName=woolf_toe_r}, {boneName=woolf_ball_r}]}]}) {layerGroup=None}
   Slot 'DefaultSlot'()
  ]}
 EventGraph {nodes [
   Event BlueprintUpdateAnimation()
   n1 = GetOwnerFortPlayerAnimInstance!(self:Custom Character Part Anim Instance Object Reference) -> (ReturnValue:Fort Player Anim Instance Object Reference)
   > Set Is Playing Custom Anim(Is Playing Custom Anim:double=GetCurveValue(self=@n1, CurveName=CustomAnim_Override)) -> (Output_Get:double)
   > n2 = Branch(Condition=NOT Boolean(A=Get bIsIdleOrSettling!(self:Fort Player Anim Instance Object Reference=@n1) -> (bIsIdleOrSettling:bool)))
   > Set ToeCurl_Left(ToeCurl_Left:double=GetCurveValue(self=@n1, CurveName=L_Toe_Splay)) -> (Output_Get:double)
   > Set ToeCurl_Right(ToeCurl_Right:double=GetCurveValue(self=@n1, CurveName=R_Toe_Splay)) -> (Output_Get:double)
   Set ToeCurl_Left(execute=@n2.else, ToeCurl_Left=0)
   > Set ToeCurl_Right(ToeCurl_Right=0)
  ]}
]
variables [
 Is Playing Custom Anim {type=number}
 ToeCurl_Left {type=number}
 ToeCurl_Right {type=number}
]
